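# M_ProtoMat is a two-sided Default Lit prototype material stored with the desert environment props.
# It exposes a green Color vector parameter and an R scalar parameter.
# A Cracks_02 mask texture is sampled with texture coordinates multiplied by a Tiling parameter.
# The graph does not show which material inputs these three branches drive.
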Material /Game/Environments/Props/Desert/Materials/M_ProtoMat
graph {
 properties {pixelDepthOffsetMode=PDOM_Legacy, twoSided=true, usedShadingModels=Default Lit}
 nodes [
  ScalarParameter() {defaultValue=0.78, parameterName=R}
  TextureSample(UVs=Multiply(A=TextureCoordinate(), B=ScalarParameter() {defaultValue=1, parameterName=Tiling})) {texture=/Game/Global/Textures/Default/Masks_and_Gradients/Cracks_02}
  VectorParameter() {parameterName=Color, defaultValue=<rgba 0.2263 0.33 0.1172 1>}
 ]
}
