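# M_TumbleWeedFx is a masked, two-sided foliage material for tumbleweed mesh particles in Cascade or Niagara.
# Particle color is both added to and multiplied with the diffuse texture, and particle alpha multiplies the mask texture's red channel.
# A normal map, an S mask texture and a scalar SS parameter complete the nodes.
# The file shows no connections to the material outputs.
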Material /Game/Environments/Props/Desert/Materials/M_TumbleWeedFx
graph {
 properties {
  bUsedWithMeshParticles=true
  bUsedWithNiagaraMeshParticles=true
  blendMode=BLEND_Masked
  pixelDepthOffsetMode=PDOM_Legacy
  shadingModel=MSM_TwoSidedFoliage
  twoSided=true
  usedShadingModels=Two Sided Foliage
 }
 nodes [
  n1 = TextureSample() {texture=/Game/Environments/Props/Desert/Textures/T_TumbleWeed_D}
  n2 = ParticleColor()
  Add(A=@n1.RGB, B=@n2.RGB)
  Multiply(A=@n2.A, B=TextureSample() {samplerType=SAMPLERTYPE_Masks, texture=/Game/Environments/Props/Desert/Textures/T_TumbleWeed_M}.R)
  Multiply(A=@n1.RGB, B=@n2.RGB)
  ScalarParameter() {defaultValue=0.5, parameterName=SS}
  TextureSample() {samplerType=SAMPLERTYPE_Normal, texture=/Game/Environments/Props/Desert/Textures/T_TumbleWeed_N}
  TextureSample() {samplerType=SAMPLERTYPE_Masks, texture=/Game/Environments/Props/Desert/Textures/T_TumbleWeed_S}
 ]
}
